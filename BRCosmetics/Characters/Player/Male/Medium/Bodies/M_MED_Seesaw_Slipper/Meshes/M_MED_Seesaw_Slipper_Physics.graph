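# Physics asset for the M_MED_Seesaw_Slipper male medium body cosmetic, covering a dangling-belt bone chain rather than a full-body ragdoll.
# It has two small capsule bodies, one on Dyn_Master_Belt_Piece1 and one on Dyn_Belt_Piece_1.
# A single constraint joins them with narrow swing and very tight twist limits, so the belt piece can move only slightly relative to the master bone.
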PhysicsAsset /BRCosmetics/Characters/Player/Male/Medium/Bodies/M_MED_Seesaw_Slipper/Meshes/M_MED_Seesaw_Slipper_Physics
bodies {
 Dyn_Belt_Piece_1 [{shapeType=Capsule, length=2, radius=1, shapeName=Dyn_Belt_Piece_1_capsule, center=<1 0 0>, rotation=<rot 90 0 0>}]
 Dyn_Master_Belt_Piece1 [{shapeType=Capsule, length=0.505, radius=0.505, shapeName=Dyn_Master_Belt_Piece1_capsule, rotation=<rot 0 -90 0>}]
}
constraints [{bone1Name=Dyn_Belt_Piece_1, bone2Name=Dyn_Master_Belt_Piece1, swing1LimitDegrees=10, swing2LimitDegrees=5, twistLimitDegrees=1}]
