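# Post-process style AnimBP for the M_MED_Seesaw_Slipper character body, correcting its shoulder pads and toggling a material element.
# It copies the parent mesh pose, then applies additive pad offsets driven by arm twist (some only while targeting), skydive floating and parachute state.
# A RigidBody node simulates the result in pelvis space.
# Weapon, customization and skydive events set "Hide Element 01" based on the tagged equipped pickaxe and skydiving.
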AnimBlueprint /BRCosmetics/Characters/Player/Male/Medium/Bodies/M_MED_Seesaw_Slipper/Meshes/M_MED_Seesaw_Slipper_AnimBP
target_skeleton=&/M_MED_Seesaw_Slipper_Skeleton
graphs [
 AnimGraph {nodes [
   "Transform (Modify) Bone - Bone: Dyn_ShoulderPad_l"(Alpha=1, Translation="1,-1.5,0", Rotation="0,-5,23", Scale="1,1,1") {
    rotationMode=BMM_Additive
    rotationSpace=BCS_BoneSpace
    translationMode=BMM_Additive
    translationSpace=BCS_BoneSpace
    alphaBoolBlend {blendOption=Linear}
    boneToModify {boneName=Dyn_ShoulderPad_l}
    rotation {roll=23, yaw=-5}
    translation=<1 -1.5>
   }
   "Transform (Modify) Bone - Bone: Dyn_ShoulderPad_r1"(Alpha=1, Translation="-1,3,-3", Rotation="0,-10,-10", Scale="1,1,1") {
    rotationMode=BMM_Additive
    rotationSpace=BCS_BoneSpace
    translationMode=BMM_Additive
    translationSpace=BCS_BoneSpace
    alphaBoolBlend {blendOption=Linear}
    boneToModify {boneName=Dyn_ShoulderPad_r1}
    rotation {roll=-10, yaw=-10}
    translation=<-1 3 -3>
   }
   "Transform (Modify) Bone - Bone: Dyn_ShoulderPad_l"(Alpha=Get Left_Shoulder Pad Down() -> (Left_Shoulder Pad Down:double), Translation="0,0,0", Rotation="0,0,0", Scale="1,1,1") {
    rotationMode=BMM_Additive
    rotationSpace=BCS_BoneSpace
    translationMode=BMM_Additive
    translationSpace=BCS_BoneSpace
    alphaBoolBlend {blendOption=Linear}
    boneToModify {boneName=Dyn_ShoulderPad_l}
   }
   n1 = "Transform (Modify) Bone - Bone: Dyn_ShoulderPad_l"(
    ComponentPose=Local To Component(LocalPose=Copy Pose From Mesh() {bUseAttachedParent=true})
    Alpha=Get Left_Shoulder pad() -> (Left_Shoulder pad:double)
    Translation="-1,-1.5,0.5"
    Rotation="2,-1,23"
    Scale="1,1,1"
   ) {
    rotationMode=BMM_Additive
    rotationSpace=BCS_BoneSpace
    translationMode=BMM_Additive
    translationSpace=BCS_BoneSpace
    alphaBoolBlend {blendOption=Linear}
    alphaScaleBiasClamp {bInterpResult=true, interpSpeedDecreasing=0.1, interpSpeedIncreasing=0.1}
    boneToModify {boneName=Dyn_ShoulderPad_l}
    rotation=<rot 2 -1 23>
    translation=<-1 -1.5 0.5>
   }
   n2 = "Transform (Modify) Bone - Bone: Dyn_ShoulderPad_r1"(
    ComponentPose=@n1
    Alpha=Get Right_Shoulder pad() -> (Right_Shoulder pad:double)
    Translation="0,1.5,-3"
    Rotation="20,-20,-25"
    Scale="1,1,1"
   ) {
    rotationMode=BMM_Additive
    rotationSpace=BCS_BoneSpace
    translationMode=BMM_Additive
    translationSpace=BCS_BoneSpace
    alphaBoolBlend {blendOption=Linear}
    alphaScaleBiasClamp {bInterpResult=true, interpSpeedDecreasing=0.1, interpSpeedIncreasing=0.1}
    boneToModify {boneName=Dyn_ShoulderPad_r1}
    rotation=<rot 20 -20 -25>
    translation {y=1.5, z=-3}
   }
   n3 = "Transform (Modify) Bone - Bone: Dyn_ShoulderPad_l"(ComponentPose=@n2, Alpha=Get Left_Shoulder Pad Down(), Translation="0,0,0", Rotation="0,-10,-10", Scale="1,1,1") {
    rotationMode=BMM_Additive
    rotationSpace=BCS_BoneSpace
    translationMode=BMM_Additive
    translationSpace=BCS_BoneSpace
    alphaBoolBlend {blendOption=Linear}
    alphaScaleBiasClamp {bInterpResult=true, interpSpeedDecreasing=0.1, interpSpeedIncreasing=0.1}
    boneToModify {boneName=Dyn_ShoulderPad_l}
    rotation {roll=-10, yaw=-10}
   }
   n4 = "Transform (Modify) Bone - Bone: Dyn_ShoulderPad_r1"(
    ComponentPose=@n3
    Alpha=Get Right_Shoulder Pad Down() -> (Right_Shoulder Pad Down:double)
    Translation="1,0,0"
    Rotation="5,-5,10"
    Scale="1,1,1"
   ) {
    rotationMode=BMM_Additive
    rotationSpace=BCS_BoneSpace
    translationMode=BMM_Additive
    translationSpace=BCS_BoneSpace
    alphaBoolBlend {blendOption=Linear}
    alphaScaleBiasClamp {bInterpResult=true, interpSpeedDecreasing=0.1, interpSpeedIncreasing=0.1}
    boneToModify {boneName=Dyn_ShoulderPad_r1}
    rotation=<rot 5 -5 10>
    translation {x=1}
   }
   n5 = "Transform (Modify) Bone - Bone: Dyn_ShoulderPad_r1"(ComponentPose=@n4, Alpha=Get RightShoulder() -> (RightShoulder:double), Translation="0,0,0", Rotation="0,0,10", Scale="1,1,1") {
    rotationMode=BMM_Additive
    rotationSpace=BCS_BoneSpace
    translationMode=BMM_Additive
    translationSpace=BCS_BoneSpace
    alphaBoolBlend {blendOption=Linear}
    alphaScaleBiasClamp {bInterpResult=true, interpSpeedDecreasing=50}
    boneToModify {boneName=Dyn_ShoulderPad_r1}
    rotation {roll=10}
   }
   n6 = Get bIsSkydiveFloating()
   n7 = "Transform (Modify) Bone - Bone: Dyn_ShoulderPad_l"!(ComponentPose=@n5, bAlphaBoolEnabled:bool=@n6, Translation="4,-1,0", Rotation="0,5,0", Scale="1,1,1") -> (Pose) {
    alphaInputType=Bool
    rotationMode=BMM_Additive
    rotationSpace=BCS_BoneSpace
    translationMode=BMM_Additive
    translationSpace=BCS_BoneSpace
    alphaBoolBlend {blendOption=Linear}
    boneToModify {boneName=Dyn_ShoulderPad_l}
    rotation {yaw=5}
    translation=<4 -1>
   }
   n8 = "Transform (Modify) Bone - Bone: Dyn_ShoulderPad_r1"!(ComponentPose=@n7, bAlphaBoolEnabled:bool=@n6, Translation="6,-1,0", Rotation="0,20,0", Scale="1,1,1") -> (Pose) {
    alphaInputType=Bool
    rotationMode=BMM_Additive
    rotationSpace=BCS_BoneSpace
    translationMode=BMM_Additive
    translationSpace=BCS_BoneSpace
    alphaBoolBlend {blendOption=Linear}
    boneToModify {boneName=Dyn_ShoulderPad_r1}
    rotation {yaw=20}
    translation=<6 -1>
   }
   n9 = Get bIsParachuteOpened()
   n10 = "Transform (Modify) Bone - Bone: Dyn_ShoulderPad_l"(ComponentPose=@n8, bAlphaBoolEnabled=@n9, Translation="4,-3,0", Rotation="0,5,0", Scale="1,1,1") {
    alphaInputType=Bool
    rotationMode=BMM_Additive
    rotationSpace=BCS_BoneSpace
    translationMode=BMM_Additive
    translationSpace=BCS_BoneSpace
    alphaBoolBlend {blendOption=Linear}
    boneToModify {boneName=Dyn_ShoulderPad_l}
    rotation {yaw=5}
    translation=<4 -3>
   }
   n11 = "Transform (Modify) Bone - Bone: Dyn_ShoulderPad_r1"(ComponentPose=@n10, bAlphaBoolEnabled=@n9, Translation="6,-3,0", Rotation="0,20,0", Scale="1,1,1") {
    alphaInputType=Bool
    rotationMode=BMM_Additive
    rotationSpace=BCS_BoneSpace
    translationMode=BMM_Additive
    translationSpace=BCS_BoneSpace
    alphaBoolBlend {blendOption=Linear}
    boneToModify {boneName=Dyn_ShoulderPad_r1}
    rotation {yaw=20}
    translation=<6 -3>
   }
   n12 = RigidBody(ComponentPose=@n11) {
    bClampLinearTranslationLimitToRefPose=true
    bOverrideWorldGravity=true
    overridePhysicsAsset=&/M_MED_Seesaw_Slipper_Physics
    simulationSpace=BaseBoneSpace
    alphaBoolBlend {blendOption=Linear}
    baseBoneRef {boneName=pelvis}
    componentAppliedLinearAccClamp=<3000 3000 3000>
    componentLinearAccScale=<0.1 0.1 0.1>
    componentLinearVelScale=<0.1 0.1 0.1>
    overrideWorldGravity {z=-980}
   }
   Output Pose(Result=Component To Local(ComponentPose=@n12)) {layerGroup=None}
   Get bIsSkyDiving()
   Get Right_Shoulder Pad Down()
  ]}
 EventGraph {nodes [
   Event BlueprintUpdateAnimation()
   > n1 = Sequence!(execute) -> (then_0, then_1, then_2:Exec, then_3:Exec, then_4:Exec, then_5:Exec)
   > Get Dynamic Material Instance(SkeletalMeshComponent=GetPartSkeletalMeshComponent())
   > GetScalarParameterValue!(execute, Collection:Material Parameter Collection Object Reference, ParameterName:Name=None) -> (then, ReturnValue:float) {is=title K2Node_CallMaterialParameterCollectionFunction}
   Set Left_Shoulder pad(
    execute=@n1.then_1
    Left_Shoulder pad:double=GetAlphaFromReferenceTransformDelta(
     SkeletalMeshComponent=Get Skeletal Mesh() -> (Skeletal Mesh:Skeletal Mesh Component Object Reference)
     BoneName=clavicle_l
     BaseName=upperarm_l
     DeltaComponentType=QuaternionTwist
     InputRange=(X=45,Y=110)
     TwistAxis="1,0,0"
    )
   ) -> (Output_Get:double)
   Set Right_Shoulder pad(
    execute=@n1.then_2
    Right_Shoulder pad:double=GetAlphaFromReferenceTransformDelta(
     SkeletalMeshComponent=Get Skeletal Mesh()
     BoneName=clavicle_r
     BaseName=upperarm_r
     DeltaComponentType=QuaternionTwist
     InputRange=(X=30,Y=126)
     TwistAxis="1,0,0"
    )
   ) -> (Output_Get:double)
   n2 = Get bIsTargeting()
   n3 = Branch(execute=@n1.then_3, Condition=@n2)
   > Set Right_Shoulder Pad Down(Right_Shoulder Pad Down:double=GetAlphaFromReferenceTransformDelta(
     SkeletalMeshComponent=Get Skeletal Mesh()
     BoneName=clavicle_r
     BaseName=upperarm_r
     DeltaComponentType=QuaternionTwist
     InputRange=(X=30,Y=-13)
     TwistAxis="1,0,0"
    )) -> (Output_Get:double)
   Set Right_Shoulder Pad Down(execute=@n3.else, Right_Shoulder Pad Down=0)
   n4 = Branch(execute=@n1.then_4, Condition=@n2)
   > Set Left_Shoulder Pad Down(Left_Shoulder Pad Down:double=GetAlphaFromReferenceTransformDelta(
     SkeletalMeshComponent=Get Skeletal Mesh()
     BoneName=clavicle_l
     BaseName=upperarm_l
     DeltaComponentType=QuaternionTwist
     InputRange=(X=55,Y=-1)
     TwistAxis="1,0,0"
    )) -> (Output_Get:double)
   Set Left_Shoulder Pad Down(execute=@n4.else, Left_Shoulder Pad Down=0)
   Set RightShoulder(
    execute=@n1.then_5
    RightShoulder:double=GetAlphaFromReferenceTransformDelta(
     SkeletalMeshComponent=Get Skeletal Mesh()
     BoneName=clavicle_r
     BaseName=upperarm_r
     DeltaComponentType=QuaternionTwist
     InputRange=(X=-24,Y=-28)
     TwistAxis="0,1,0"
    )
   ) -> (Output_Get:double)
   Event BlueprintInitializeAnimation()
   > Set Skeletal Mesh(Skeletal Mesh:Skeletal Mesh Component Object Reference=GetOwnerSkeletalMeshComponent()) -> (Output_Get:Skeletal Mesh Component Object Reference)
   n5 = Event OnWeaponChanged()
   n6 = Get Dynamic Material Instance(execute=[@n5.then, @n14.then], SkeletalMeshComponent=GetPartSkeletalMeshComponent())
   > n7 = Sequence!(execute) -> (then_0, then_1)
   n8 = PickaxeItemDefinitionHasAnyMetaTag(Tags=Make Array("[0]"=(TagName="Cosmetics.MetaTags.Seesaw_Slipper")), bRequiresEquipped=true)
   > n9 = Branch(Condition=@n8)
   > n10 = Branch(Condition=Get bIsSkyDiving())
   SetScalarParameterValue(execute=[@n12.then, @n10.then, @n11.else, @n9.else], self=@n6, ParameterName=Hide Element 01, Value=0)
   SetScalarParameterValue(execute=[@n10.else, @n11.then], self=@n6, ParameterName=Hide Element 01, Value=1)
   n11 = Branch(execute=[@n13.then, @n7.then_1], Condition=@n8)
   n12 = Event OnBeginSkydiving()
   n13 = Event OnEndSkydiving()
   n14 = Event OnCharacterCustomizationLoaded()
   Set Right_Shoulder pad(Right_Shoulder pad=0)
  ]}
]
variables [
 Fort pawn {type=FortPawn, default=null}
 Left_Shoulder pad {type=number}
 Left_Shoulder Pad Down {type=number}
 LeftShoulder {type=number}
 Right_Shoulder pad {type=number}
 Right_Shoulder Pad Down {type=number}
 RightShoulder {type=number}
 Skeletal Mesh {type=SkeletalMeshComponent}
]
